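# First-person sound cue for HVAC activation, apparently heard when the vent launches the player.
# A Random node picks one of three equally weighted player-launch waves and plays it at reduced volume in the Fort_Gadgets sound class.
# Its attenuation override turns off distance falloff and spatialization, superseding the assigned golf-cart attenuation asset so the sound plays as non-positional audio.
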SoundCue /Clyde/Athena/Sounds/Ambient/HVAC/HVAC_Activate_1P_Cue
root=Random_1
nodes [
 {id=Random_1, children [WavePlayer_3,WavePlayer_4,WavePlayer_5], properties {weights [1,1,1]}}
 {id=WavePlayer_3, properties {soundWaveAssetPtr=/Game/Athena/Sounds/Ambient/HVAC/HVAC_PlayerLaunch_01}}
 {id=WavePlayer_4, properties {soundWaveAssetPtr=/Game/Athena/Sounds/Ambient/HVAC/HVAC_PlayerLaunch_02}}
 {id=WavePlayer_5, properties {soundWaveAssetPtr=/Game/Athena/Sounds/Ambient/HVAC/HVAC_PlayerLaunch_03}}
]
properties {
 attenuationSettings=/Game/Sounds/Attenuation/Fort_GolfCart_Movement_Distant_Att
 bOverrideAttenuation=true
 maxDistance=2097152
 soundClassObject=&Fort_Gadgets
 volumeMultiplier=0.5
 attenuationOverrides {bAttenuate=false, bSpatialize=false}
}
