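# Sound cue for HVAC activation, built by layering three randomized one-shot groups through a mixer.
# A boosted RatRod hammer-jump layer sits over a quieter HVAC launch layer that passes through a modulator for variation.
# A geyser launch layer runs through an enveloper that fades its volume out over the first second.
# The cue reuses the golf cart distant-movement attenuation and plays in the gadgets sound class.
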SoundCue /Clyde/Athena/Sounds/Ambient/HVAC/HVAC_Activate_Cue
root=Mixer_0
nodes [
 {
  id=Enveloper_0
  children [Random_3]
  properties {pitchCurve {editorCurveData {keys [{time=0}]}}, volumeCurve {editorCurveData {keys [<at 0 1>,<at 1 0>]}}}
 }
 {id=Mixer_0, children [Random_2,Modulator_0,Enveloper_0], properties {inputVolume [1.2,0.3,1]}}
 {id=Modulator_0, children [Random_1]}
 {id=Random_1, children [WavePlayer_0,WavePlayer_1,WavePlayer_2], properties {weights [1,1,1]}}
 {id=Random_2, children [WavePlayer_7,WavePlayer_8], properties {weights [1,1]}}
 {id=Random_3, children [WavePlayer_9, WavePlayer_10, WavePlayer_11], properties {weights [1,1,1]}}
 {id=WavePlayer_0, properties {soundWaveAssetPtr=/Game/Athena/Sounds/Ambient/HVAC/HVAC_Launch_01}}
 {id=WavePlayer_1, properties {soundWaveAssetPtr=/Game/Athena/Sounds/Ambient/HVAC/HVAC_Launch_02}}
 {id=WavePlayer_10, properties {soundWaveAssetPtr=/Game/Athena/Sounds/Ambient/Geysers/Geyser_Launch_1P_02}}
 {id=WavePlayer_11, properties {soundWaveAssetPtr=/Game/Athena/Sounds/Ambient/Geysers/Geyser_Launch_1P_03}}
 {id=WavePlayer_2, properties {soundWaveAssetPtr=/Game/Athena/Sounds/Ambient/HVAC/HVAC_Launch_03}}
 {id=WavePlayer_7, properties {soundWaveAssetPtr=/Game/Sounds/Fort_Weapons/RatRod/Hammer/RR_Hammer_Jump_02}}
 {id=WavePlayer_8, properties {soundWaveAssetPtr=/Game/Sounds/Fort_Weapons/RatRod/Hammer/RR_Hammer_Jump_01}}
 {id=WavePlayer_9, properties {soundWaveAssetPtr=/Game/Athena/Sounds/Ambient/Geysers/Geyser_Launch_1P_01}}
]
properties {
 attenuationSettings=/Game/Sounds/Attenuation/Fort_GolfCart_Movement_Distant_Att
 maxDistance=10100
 soundClassObject=&Fort_Gadgets
}
