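# Sound cue for ambient HVAC loops, routed to the Fort_Gadgets sound class with reduced volume and a small concurrency cap.
# A random node picks one of three equally weighted looping waves, then passes it through a default-setting modulator.
# An attenuation node adds distance low-pass filtering, listener focus and occlusion, with reverb send disabled.
# The root mixer takes that branch; its second input is empty.
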SoundCue /Clyde/Athena/Sounds/Ambient/HVAC/HVAC_Loop_Cue
root=Mixer_1
nodes [
 {
  id=Attenuation_1
  children [Modulator_0]
  properties {
   bOverrideAttenuation=true
   attenuationOverrides {
    bAttenuateWithLPF=true
    bEnableFocusInterpolation=true
    bEnableListenerFocus=true
    bEnableOcclusion=true
    bEnableReverbSend=false
    falloffDistance=1500
    focusReleaseInterpSpeed=0.5
    lPFFrequencyAtMax=2000
    lPFRadiusMax=3000
    lPFRadiusMin=500
    nonFocusVolumeAttenuation=0.5
    occlusionInterpolationTime=1
    occlusionLowPassFilterFrequency=2000
    occlusionTraceChannel=ECC_GameTraceChannel11
    occlusionVolumeAttenuation=0.5
    stereoSpread=512
    attenuationShapeExtents {x=500}
   }
  }
 }
 {id=Mixer_1, children [Attenuation_1, null], properties {inputVolume [1,1]}}
 {id=Modulator_0, children [Random_1]}
 {id=Random_1, children [WavePlayer_7,WavePlayer_8,WavePlayer_9], properties {weights [1,1,1]}}
 {id=WavePlayer_7, properties {bLooping=true, soundWaveAssetPtr=/Game/Athena/Sounds/Ambient/HVAC/HVAC_Loop_01}}
 {id=WavePlayer_8, properties {bLooping=true, soundWaveAssetPtr=/Game/Athena/Sounds/Ambient/HVAC/HVAC_Loop_02}}
 {id=WavePlayer_9, properties {bLooping=true, soundWaveAssetPtr=/Game/Athena/Sounds/Ambient/HVAC/HVAC_Loop_04}}
]
properties {
 bOverrideConcurrency=true
 maxDistance=2000
 soundClassObject=&Fort_Gadgets
 volumeMultiplier=0.4
 attenuationOverrides {attenuationShapeExtents {x=0}}
 concurrencyOverrides {maxCount=3}
}
